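# PistolCanary hab1000 audio Level Sequence, which times audio manager changes, apparently for a cinematic or scripted moment.
# An event track on the possessed PistolCanary_AudioManager fires two triggers, whose target functions are not shown.
# A Hero Active Mix bool track switches on just after the first trigger and off just before the second.
# A TentacleBoss (BGA_Tentacle) binding is present but has no tracks.
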LevelSequence /PistolCanary/Sequence/hab/hab1000/PistolCanary_hab1000_aud
bindings [
 PistolCanary_AudioManager {
  id=#1
  possessed_class=PistolCanary_AudioManager_C
  tracks [
   {
    type=MovieSceneEventTrack
    sections [{
      type=MovieSceneEventTriggerSection
      channels [None {
        type=MovieSceneScriptingEventChannel
        keys [[3600, "<Struct 'MovieSceneEvent' (0x852f46aaa00) {}>"], [4075, "<Struct 'MovieSceneEvent' (0x852f46aa000) {}>"]]
       }]
     }]
   }
   {
    type=MovieSceneBoolTrack
    display_name=Hero Active Mix
    sections [{
      type=MovieSceneBoolSection
      channels [None {type=MovieSceneScriptingBoolChannel, default=false, keys [[3466,false],[3601,true],[4074,false]]}]
     }]
   }
  ]
 }
 TentacleBoss {id=#2, possessed_class=BGA_Tentacle_C, tracks []}
]
display_rate {denominator=1, numerator=60}
tick_resolution {denominator=1, numerator=24000}
tracks []
