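# PistolCanary_hab1000_gam is a Level Sequence for hab1000, apparently the gameplay layer judging by the gam suffix.
# It binds no objects and holds a single event track with one trigger section.
# That section fires two closely spaced events on a 60 fps timeline, but the functions they call are not shown.
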LevelSequence /PistolCanary/Sequence/hab/hab1000/PistolCanary_hab1000_gam
bindings []
display_rate {denominator=1, numerator=60}
tick_resolution {denominator=1, numerator=24000}
tracks [{
  type=MovieSceneEventTrack
  sections [{
    type=MovieSceneEventTriggerSection
    channels [None {
      type=MovieSceneScriptingEventChannel
      keys [[3716, "<Struct 'MovieSceneEvent' (0x852f46aac80) {}>"], [4062, "<Struct 'MovieSceneEvent' (0x852f46aabe0) {}>"]]
     }]
   }]
 }]
