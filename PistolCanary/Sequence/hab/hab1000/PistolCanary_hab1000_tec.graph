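# PistolCanary hab1000 "tec" Level Sequence that times panel events and a comic-intro post-process effect.
# An event track fires three times on PistolCanary_PanelManager, starting before the effect appears.
# The scaled, statically placed PC_Comic_Intro_PP actor is toggled visible for a window.
# Within it, its PostProcess component's Blend Weight goes from full to zero, while five array weights rise early and return to zero, the fourth rising later and holding.
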LevelSequence /PistolCanary/Sequence/hab/hab1000/PistolCanary_hab1000_tec
bindings [
 PistolCanary_PanelManager {
  id=#1
  possessed_class=PistolCanary_PanelManager_C
  tracks [{
    type=MovieSceneEventTrack
    sections [{
      type=MovieSceneEventTriggerSection
      channels [None {
        type=MovieSceneScriptingEventChannel
        keys [
         [3601, "<Struct 'MovieSceneEvent' (0x852f46ab5e0) {}>"]
         [3685, "<Struct 'MovieSceneEvent' (0x852f46ab680) {}>"]
         [3779, "<Struct 'MovieSceneEvent' (0x852f46ab720) {}>"]
        ]
       }]
     }]
   }]
 }
 PostProcess {
  id=#2
  parent_id=#3
  possessed_class=PostProcessComponent
  tracks [
   {
    type=MovieSceneFloatTrack
    display_name=Blend Weight
    sections [{type=MovieSceneFloatSection, channels [None {type=MovieSceneScriptingFloatChannel, default=1, keys [[3700,1],[4076,0]]}]}]
   }
   {
    type=MovieSceneFloatTrack
    display_name="Weight (Array[0])"
    sections [{type=MovieSceneFloatSection, channels [None {type=MovieSceneScriptingFloatChannel, default=0, keys [[3700,0],[3716,1],[4062,0]]}]}]
   }
   {
    type=MovieSceneFloatTrack
    display_name="Weight (Array[1])"
    sections [{type=MovieSceneFloatSection, channels [None {type=MovieSceneScriptingFloatChannel, default=0, keys [[3700,0],[3716,1],[4062,0]]}]}]
   }
   {
    type=MovieSceneFloatTrack
    display_name="Weight (Array[2])"
    sections [{type=MovieSceneFloatSection, channels [None {type=MovieSceneScriptingFloatChannel, default=1, keys [[3700,0],[3716,1],[4062,0]]}]}]
   }
   {
    type=MovieSceneFloatTrack
    display_name="Weight (Array[3])"
    sections [{
      type=MovieSceneFloatSection
      channels [None {type=MovieSceneScriptingFloatChannel, default=0, keys [[3700,0],[3716,0],[3731,1],[4047,1],[4062,0]]}]
     }]
   }
   {
    type=MovieSceneFloatTrack
    display_name="Weight (Array[4])"
    sections [{type=MovieSceneFloatSection, channels [None {type=MovieSceneScriptingFloatChannel, default=0, keys [[3700,0],[3716,1],[4062,0]]}]}]
   }
  ]
 }
 PC_Comic_Intro_PP {
  id=#3
  possessed_class=PC_Comic_Intro_PP_C
  tracks [
   {
    type=MovieSceneVisibilityTrack
    sections [{
      type=MovieSceneVisibilitySection
      channels [None {type=MovieSceneScriptingBoolChannel, default=false, keys [[3601,false],[3700,true],[4076,false]]}]
     }]
   }
   {
    type=MovieScene3DTransformTrack
    sections [{
      type=MovieScene3DTransformSection
      channels [
       Location.X {type=MovieSceneScriptingDoubleChannel, default=14488}
       Location.Y {type=MovieSceneScriptingDoubleChannel, default=72664}
       Location.Z {type=MovieSceneScriptingDoubleChannel, default=3724}
       Rotation.X {type=MovieSceneScriptingDoubleChannel, default=0}
       Rotation.Y {type=MovieSceneScriptingDoubleChannel, default=0}
       Rotation.Z {type=MovieSceneScriptingDoubleChannel, default=0}
       Scale.X {type=MovieSceneScriptingDoubleChannel, default=350}
       Scale.Y {type=MovieSceneScriptingDoubleChannel, default=350}
       Scale.Z {type=MovieSceneScriptingDoubleChannel, default=350}
      ]
     }]
   }
  ]
 }
]
display_rate {denominator=1, numerator=60}
tick_resolution {denominator=1, numerator=24000}
tracks []
