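# Level Sequence for the PistolCanary hab1000 VFX pass, driving the AshenMagnus attack Niagara effect.
# A spawn track brings in the attached Niagara actor for a limited frame window, with a system spawn section running on its component.
# Niagara parameter tracks time the attack by keying AttackProgress, AttackActive and the TextSpawn and FlashSpawn values, while BulletSource supplies a constant vector.
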LevelSequence /PistolCanary/Sequence/hab/hab1000/PistolCanary_hab1000_vfx
bindings [
 NS_PistolCanary_AshenMagnus_Attack {
  id=#1
  possessed_class=NiagaraActor
  tracks [
   {
    type=MovieSceneSpawnTrack
    sections [{
      type=MovieSceneSpawnSection
      channels [None {type=MovieSceneScriptingBoolChannel, default=false, keys [[3600,false],[3776,true],[3968,false]]}]
     }]
   }
   {
    type=MovieScene3DTransformTrack
    sections [{
      type=MovieScene3DTransformSection
      channels [
       Location.X {type=MovieSceneScriptingDoubleChannel, default=0}
       Location.Y {type=MovieSceneScriptingDoubleChannel, default=0}
       Location.Z {type=MovieSceneScriptingDoubleChannel, default=0}
       Rotation.X {type=MovieSceneScriptingDoubleChannel, default=0}
       Rotation.Y {type=MovieSceneScriptingDoubleChannel, default=0}
       Rotation.Z {type=MovieSceneScriptingDoubleChannel, default=0}
       Scale.X {type=MovieSceneScriptingDoubleChannel, default=1}
       Scale.Y {type=MovieSceneScriptingDoubleChannel, default=1}
       Scale.Z {type=MovieSceneScriptingDoubleChannel, default=1}
      ]
     }]
   }
   {type=MovieScene3DAttachTrack, sections [{type=MovieScene3DAttachSection}]}
  ]
 }
 NiagaraComponent0 {
  id=#2
  parent_id=#1
  possessed_class=NiagaraComponent
  tracks [
   {
    type=MovieSceneNiagaraFloatParameterTrack
    display_name=AttackProgress
    sections [{type=MovieSceneFloatSection, channels [None {type=MovieSceneScriptingFloatChannel, default=0.2, keys [[3776,0.1],[3897,0.2]]}]}]
   }
   {
    type=MovieSceneNiagaraFloatParameterTrack
    display_name=AttackActive
    sections [{type=MovieSceneFloatSection, channels [None {type=MovieSceneScriptingFloatChannel, default=0, keys [[3883,1],[3893,0]]}]}]
   }
   {
    type=MovieSceneNiagaraVectorParameterTrack
    display_name=BulletSource
    sections [{
      type=MovieSceneFloatVectorSection
      channels [
       Vector.X {type=MovieSceneScriptingFloatChannel, default=11000}
       Vector.Y {type=MovieSceneScriptingFloatChannel, default=75000}
       Vector.Z {type=MovieSceneScriptingFloatChannel, default=11000}
      ]
     }]
   }
   {
    type=MovieSceneNiagaraFloatParameterTrack
    display_name=TextSpawn
    sections [{type=MovieSceneFloatSection, channels [None {type=MovieSceneScriptingFloatChannel, default=0, keys [[3837,3],[3879,4],[3888,0]]}]}]
   }
   {
    type=MovieSceneNiagaraFloatParameterTrack
    display_name=FlashSpawn
    sections [{type=MovieSceneFloatSection, channels [None {type=MovieSceneScriptingFloatChannel, default=0, keys [[3866,10],[3869,0]]}]}]
   }
   {type=MovieSceneNiagaraSystemTrack, sections [{type=MovieSceneNiagaraSystemSpawnSection, end_frame=4379, start_frame=3779}]}
  ]
 }
]
display_rate {denominator=1, numerator=60}
tick_resolution {denominator=1, numerator=24000}
tracks []
